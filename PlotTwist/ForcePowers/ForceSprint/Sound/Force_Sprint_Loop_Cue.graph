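# Looping wind sound cue for the Force Sprint power, set in the player foley sound class with a small concurrency limit.
# A looping wind wave runs through a lowered-pitch random modulator, a sound class node overriding to third-person ranged weapon fire, and a wind attenuation node.
# That chain enters the root mixer at half volume beside an empty input, and the cue's pitch multiplier lowers pitch further.
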SoundCue /PlotTwist/ForcePowers/ForceSprint/Sound/Force_Sprint_Loop_Cue
root=Mixer_0
nodes [
 {id=Attenuation_2, children [SoundClass_1], properties {attenuationSettings=/Game/Sounds/Attenuation/Fort_ForcePowers_ForceSprint_Wind_Att}}
 {id=Mixer_0, children [null, Attenuation_2], properties {inputVolume [2,0.5]}}
 {id=Modulator_1, children [WavePlayer_0], properties {pitchMax=0.8, pitchMin=0.6, volumeMax=1, volumeMin=1}}
 {id=SoundClass_1, children [Modulator_1], properties {soundClassOverride=&Fort_WeaponRanged_Fire_3P_PSM}}
 {id=WavePlayer_0, properties {bLooping=true, soundWaveAssetPtr=&/Force_Sprint_Wind_Loop}}
]
properties {
 bOverrideConcurrency=true
 maxDistance=5500
 pitchMultiplier=0.5
 soundClassObject=&Fort_Player_Foley
 volumeMultiplier=1
 attenuationOverrides {falloffDistance=6000, attenuationShapeExtents {x=1000}}
 concurrencyOverrides {maxCount=3}
}
