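# Sound cue for the start of the Force Sprint power, playing one of four equally weighted "Force_Sprint_On" waves at random.
# The pick passes through a pitch-varying modulator with fixed volume, then a sound class node and an attenuation node using the Force Sprint settings.
# The node overrides the cue's Player Foley class with a third-person ranged weapon fire class.
# Overridden concurrency caps simultaneous instances at three.
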SoundCue /PlotTwist/ForcePowers/ForceSprint/Sound/Force_Sprint_Start_Cue
root=Attenuation_0
nodes [
 {id=Attenuation_0, children [SoundClass_0], properties {attenuationSettings=/Game/Sounds/Attenuation/Fort_ForcePowers_ForceSprint_Att}}
 {id=Modulator_0, children [Random_0], properties {pitchMax=1.2, pitchMin=0.8, volumeMax=1, volumeMin=1}}
 {id=Random_0, children [WavePlayer_4,WavePlayer_5,WavePlayer_6,WavePlayer_7], properties {weights [1,1,1,1]}}
 {id=SoundClass_0, children [Modulator_0], properties {soundClassOverride=&Fort_WeaponRanged_Fire_3P_PSM}}
 {id=WavePlayer_4, properties {soundWaveAssetPtr=&/Force_Sprint_On_01}}
 {id=WavePlayer_5, properties {soundWaveAssetPtr=&/Force_Sprint_On_02}}
 {id=WavePlayer_6, properties {soundWaveAssetPtr=&/Force_Sprint_On_03}}
 {id=WavePlayer_7, properties {soundWaveAssetPtr=&/Force_Sprint_On_04}}
]
properties {
 bOverrideConcurrency=true
 maxDistance=5500
 soundClassObject=&Fort_Player_Foley
 volumeMultiplier=1
 attenuationOverrides {falloffDistance=6000, attenuationShapeExtents {x=1000}}
 concurrencyOverrides {maxCount=3}
}
